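# UI material for icons in the Creative property editor's icon picker, drawing a tinted icon over a colored circular background.
# The icon texture is scaled about its center, and its red channel is squared, masked to the UV bounds and smoothstepped into a shape that blends background color to icon color.
# A smoothstepped radial gradient forms the background disc, scaled by background alpha and maxed with the icon shape to form a combined translucency mask.
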Material /Game/Athena/UI/InGame/Creative/PropertyEditor/Icon_Picker/Materials/M_UI_IconPicker_Icon
graph {
 properties {
  bUsedWithStaticMesh=false
  blendMode=BLEND_Translucent
  materialDomain=MD_UI
  pixelDepthOffsetMode=PDOM_Legacy
  usedShadingModels=Default Lit
 }
 nodes [
  n1 = VectorParameter() {parameterName=BGcolor, defaultValue=<rgba 0.0789 0.4612 0.7969 1>}
  n2 = MaterialFunctionCall(Texture Scale=ScalarParameter() {defaultValue=1, parameterName=IconScale}) {materialFunction=&ScaleUVsByCenter}
  n3 = TextureSampleParameter2D(UVs=@n2.UVs) {parameterName=Texture, texture=/Game/Athena/UI/InGame/Creative/PropertyEditor/Icon_Picker/Textures/T_UI_IconPicker_NONE}
  n4 = ScalarParameter() {defaultValue=0.2, parameterName=IconEdge}
  n5 = MaterialFunctionCall(
   Alpha=Multiply(A=Multiply(A=@n3.R, B=@n3.R), B=@n2.0-1 mask)
   Min=@n4
   Max=Add(A=@n4, B=ScalarParameter() {defaultValue=0.01, parameterName=IconEdgeSoftness})
  ) {materialFunction=&SmoothStep}
  LinearInterpolate(A=@n1.RGB, B=VectorParameter() {parameterName=IconColor, defaultValue=<rgba 1 1 1 1>}.RGB, Alpha=@n5.Result)
  n6 = ScalarParameter() {defaultValue=0.1, parameterName=BGEdge}
  n7 = MaterialFunctionCall(
   Alpha=MaterialFunctionCall() {materialFunction=&RadialGradientExponential}.RadialGradientExponential
   Min=@n6
   Max=Add(A=@n6, B=ScalarParameter() {defaultValue=0.01, parameterName=BGEdgeSoftness})
  ) {materialFunction=&SmoothStep}
  Reroute(None=Max(A=@n5.Result, B=Multiply(A=@n1.A, B=@n7.Result)))
 ]
}
